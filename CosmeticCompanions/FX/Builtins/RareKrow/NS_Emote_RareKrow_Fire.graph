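# Niagara system for the RareKrow companion's fire emote, projecting a flame stream from the skeletal mesh and leaving a wall of fire.
# User.Reveal drives the stream's brightness, opacity, ember rate and sweeping end point, with looping flipbook sprites placed along a Hermite spline.
# FireWall001 uses spawn groups to layer several sprite types and a scorch decal in one emitter.
# FireWall002 bursts flames along a line with index-staggered flipbook timing.
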NiagaraSystem /CosmeticCompanions/FX/Builtins/RareKrow/NS_Emote_RareKrow_Fire
dependencies {
 usedDataInterfaces [
  &NiagaraDataInterfaceCurve
  &NiagaraDataInterfaceSpriteRendererInfo
  &NiagaraDataInterfaceColorCurve
  &NiagaraDataInterfaceVectorCurve
  &NiagaraDataInterfaceArrayFloat
  &NiagaraDataInterfaceVector2DCurve
  &NiagaraDataInterfaceSkeletalMesh
 ]
 usedDynamicInputs [&FNTimeOfDayScalar]
 usedModules [
  &self:SystemSpawnScript.SystemScriptSource.SystemScriptGraph.NiagaraNodeAssignment_1.SetVariables_#1
  &self:SystemSpawnScript.SystemScriptSource.SystemScriptGraph.NiagaraNodeAssignment_2.SetVariables_#2
  &SystemState
  &BlendSpaceHelper
  &InterpolateOverTime
  &self:SystemSpawnScript.SystemScriptSource.SystemScriptGraph.NiagaraNodeAssignment_0.SetVariables_#3
 ]
}
emitters {
 AdditionalFlames001 {
  emitterSpawnScript [/SetVariables(Emitter.Length=166, Emitter.ParticlesToSpawn=16) {isSetParameters=true, module=SetVariables_#4}]
  emitterUpdateScript [
   &EmitterState(Inactive Response=::NewEnumerator1, Life Cycle Mode=::NewEnumerator1, Loop Behavior=::NewEnumerator1, UseLoopDelay=false)
   &SpawnBurst_Instantaneous(Spawn Count=$Emitter.ParticlesToSpawn, Spawn Probability=0.5, Use Spawn Probability=false, UseLoopCountLimit=false)
   /SetVariables(Emitter.Reveal:NiagaraFloat=$User.Reveal) {isSetParameters=true, module=SetVariables_#5}
   /SetVariables(Emitter.Length:NiagaraFloat=&Multiply_Float(
     A=$Emitter.Initial.Length
     B=&FloatFromCurve(
      FloatCurve=DataInterface() {propertyValues {Curve {keys [{interpMode=RCIM_Cubic, leaveTangent=2.4356, tangentMode=RCTM_Break, time=0, value=0}, <at 1 1>]}}}
      CurveIndex=$Emitter.Reveal
     )
    )) {isSetParameters=true, module=SetVariables_#6}
  ]
  particleSpawnScript [&InitializeParticle(
    Color=<rgba 3 3 3 1>
    Sprite Rotation Angle=&UniformRangedFloat(Minimum=-30, Maximum=30)
    Sprite Rotation Mode=::NewEnumerator0
    Sprite Size=&Multiply_Vector2D_ByFloat(Float=&RandomRangeFloat(Minimum=50, Maximum=50), Vector2D=<0.2 2>)
    Sprite Size Mode=::NewEnumerator4
    Sprite UV Mode=::NewEnumerator1
    UsePositionOffset=false
    Write Material Random=true
   )]
  particleUpdateScript [
   &ParticleState(Kill Particles When Lifetime Has Elapsed=false)
   /SetVariables(Particles.ParticleRelativeLifetimeSprites:NiagaraFloat=&Frac_Float(Float=&Add_Float(A=&Multiply_Float(A=1.5, B=$Emitter.Age), B=&ReturnNormalizedExecIndex()))) {isSetParameters=true, module=SetVariables_#7}
   /SetVariables(Particles.OpacityMultiplierSprites:NiagaraFloat=&FloatFromCurve(
     FloatCurve=DataInterface() {propertyValues {Curve {keys [
         {interpMode=RCIM_Cubic, tangentMode=RCTM_User, time=0, value=0}
         {interpMode=RCIM_Cubic, tangentMode=RCTM_User, time=0.5, value=1}
         {arriveTangent=-2, interpMode=RCIM_Cubic, leaveTangent=-2, time=0.927, value=0.7949}
         {tangentMode=RCTM_User, time=1, value=0}
        ]}}}
     CurveIndex=&Add_Float(
      A=&Multiply_Float(A=$Particles.ParticleRelativeLifetimeSprites, B=-1)
      B=&Multiply_Float(
       A=&FloatFromCurve(
        FloatCurve=DataInterface() {propertyValues {Curve {
           defaultValue=0
           keys [{tangentMode=RCTM_User, time=0, value=0}, <at 0.4638 0.5352>, <at 0.8935 0.5814>, {tangentMode=RCTM_User, time=1, value=1}]
          }}}
        CurveIndex=$Emitter.Reveal
       )
       B=2
      )
     )
    )) {isSetParameters=true, module=SetVariables_#8}
   /SetVariables(Particles.InterpolateCurveSprites:NiagaraFloat=&FloatFromCurve(
     FloatCurve=DataInterface() {propertyValues {Curve {keys [
         {arriveTangent=-0.6503, interpMode=RCIM_Cubic, leaveTangent=1.4736, tangentMode=RCTM_Break, time=0, value=0}
         {arriveTangent=0.4202, interpMode=RCIM_Cubic, tangentMode=RCTM_Break, time=1, value=1}
        ]}}}
     CurveIndex=$Particles.ParticleRelativeLifetimeSprites
    )) {isSetParameters=true, module=SetVariables_#9}
   /SetVariables(Particles.PositionSprites:NiagaraPosition=&AddVectorToPosition(
     Position=&AddVectorToPosition(
      Position=/Niagara/DynamicInputs/Helpers/SimulationPosition()
      Vector=&TransformVector(Vector=<0 1 0>, Destination Space=::Simulation)
     )
     Vector=&Multiply_VectorByFloat(
      Vector=&RotateVectorByQuaternion(
       Vector To Rotate=<0 1 0>
       Quaternion=&LerpQuaternion(
        Quaternion A:Quat4f=$Engine.Owner.Rotation
        Quaternion B:Quat4f=$System.InterpolateOverTime.InterpolateOverTime_Quaternion.Moving Average
        Lerp Factor:NiagaraFloat=$Particles.InterpolateCurveSprites
        Enable Spherical Linear Interpolation:NiagaraBool=false
       )
      )
      Float=&Multiply_Float(A=$Emitter.Length, B=$Particles.InterpolateCurveSprites)
     )
    )) {enabled=false, isSetParameters=true, module=SetVariables_#10}
   /SetVariables(Particles.SpriteSize=&Multiply_Vector2D(
     A=&Vector2DFromFloat(Value=&FloatFromCurve(
       FloatCurve=DataInterface() {propertyValues {Curve {keys [
           {arriveTangent=0.5131, interpMode=RCIM_Cubic, leaveTangent=0.1835, tangentMode=RCTM_Break, time=0, value=0.4}
           {interpMode=RCIM_Cubic, tangentMode=RCTM_Break, time=0.6, value=1}
           {arriveTangent=-3.5982, tangentMode=RCTM_User, time=1, value=0.5}
          ]}}}
       CurveIndex=$Particles.InterpolateCurveSprites
       Scale Curve=&Lerp_Float(A=55, B=70, Alpha=&RandomRangeFloat(Minimum=0, Maximum=1))
      ))
     B=<1 1>
    )) {isSetParameters=true, module=SetVariables_#11}
   &SubUVAnimation(
    "Sprite Renderer "=DataInterface() {propertyValues {SpriteRenderer=&self:AdditionalFlames001_0.NiagaraSpriteRendererProperties_2}}
    SubUV Lookup Index:NiagaraFloat=$Particles.ParticleRelativeLifetimeSprites
    UseEndFrame=true
    UseStartFrame=true
   )
   &ScaleColor(
    Scale Alpha=&Multiply_Float(
     A=&Multiply_Float(
      A=&FloatFromCurve(
       FloatCurve=DataInterface() {propertyValues {Curve {
          defaultValue=0
          keys [
           {interpMode=RCIM_Cubic, tangentMode=RCTM_User, time=0, value=0}
           {interpMode=RCIM_Cubic, tangentMode=RCTM_User, time=0.5, value=1}
           {tangentMode=RCTM_User, time=1, value=0}
          ]
         }}}
       CurveIndex=$Particles.ParticleRelativeLifetimeSprites
      )
      B=&Multiply_Float(
       A=&RandomRangeFloat(Minimum=1, Maximum=1)
       B=&UniformAOrBFloat(A=0, Distribution Weight=1, Evaluation Type=::NewEnumerator0, Recalculate Random Each Loop=true, Override Seed=false)
      )
     )
     B=$Particles.OpacityMultiplierSprites
    )
    Scale RGB=&Multiply_VectorByFloat(
     Vector=&Multiply_VectorByFloat(
      Vector=<1 1 1>
      Float=&FloatFromCurve(
       FloatCurve=DataInterface() {propertyValues {Curve {keys [{arriveTangent=-5.044, interpMode=RCIM_Cubic, leaveTangent=-5.044, tangentMode=RCTM_User, time=0.25, value=2}, <at 0.4 1 RCIM_Cubic>]}}}
       CurveIndex=$User.Reveal
      )
     )
     Float=&FloatFromCurve(
      FloatCurve=DataInterface() {propertyValues {Curve {keys [
          {arriveTangent=9.0725, interpMode=RCIM_Cubic, time=0, value=0}
          <at 0.5 0.9 RCIM_Cubic>
          {arriveTangent=-4.8128, interpMode=RCIM_Cubic, leaveTangent=-4.8128, tangentMode=RCTM_User, time=1, value=0}
         ]}}}
      CurveIndex=$Particles.ParticleRelativeLifetimeSprites
      Scale Curve=&Multiply_Float(A=&RandomRangeFloat(Minimum=0.5, Maximum=1), B=1)
     )
    )
    ScaleA=true
    ScaleRGB=true
   ) {enabled=false}
   &ScaleColor(
    Scale Alpha=&Multiply_Float(
     A=&Multiply_Float(
      A=&FloatFromCurve(
       FloatCurve=DataInterface() {propertyValues {Curve {
          defaultValue=0
          keys [
           {interpMode=RCIM_Cubic, tangentMode=RCTM_User, time=0, value=0}
           {interpMode=RCIM_Cubic, tangentMode=RCTM_User, time=0.5, value=1}
           {arriveTangent=-6.148, tangentMode=RCTM_User, time=1, value=0}
          ]
         }}}
       CurveIndex=$Particles.ParticleRelativeLifetimeSprites
      )
      B=&Multiply_Float(
       A=&RandomRangeFloat(Minimum=1, Maximum=1)
       B=&UniformAOrBFloat(A=0, Distribution Weight=1, Evaluation Type=::NewEnumerator0, Recalculate Random Each Loop=true, Override Seed=false)
      )
     )
     B=$Particles.OpacityMultiplierSprites
    )
    Scale RGB=&Multiply_VectorByFloat(
     Vector=&Multiply_VectorByFloat(
      Vector=<1 1 1>
      Float=&FloatFromCurve(
       FloatCurve=DataInterface() {propertyValues {Curve {keys [
           <at 0 7.6621 RCIM_Cubic>
           {arriveTangentWeight=0.3527, interpMode=RCIM_Cubic, leaveTangentWeight=1.1383, tangentMode=RCTM_User, time=0.1953, value=1.9863}
           {arriveTangentWeight=1.1221, interpMode=RCIM_Cubic, leaveTangentWeight=0.067, tangentMode=RCTM_User, time=0.85, value=2}
           {arriveTangent=-12.4245, interpMode=RCIM_Cubic, leaveTangent=-12.4245, tangentMode=RCTM_User, time=0.8891, value=1.3527}
           <at 1 1 RCIM_Cubic>
          ]}}}
       CurveIndex=$User.Reveal
      )
     )
     Float=&FloatFromCurve(
      FloatCurve=DataInterface() {propertyValues {Curve {keys [
          {arriveTangent=9.0725, interpMode=RCIM_Cubic, time=0, value=0}
          {arriveTangent=0.7472, interpMode=RCIM_Cubic, leaveTangent=0.7472, time=0.5, value=0.9}
          {arriveTangent=-1.8, interpMode=RCIM_Cubic, leaveTangent=-1.8, time=0.9447, value=0.7059}
          {arriveTangent=-4.8128, interpMode=RCIM_Cubic, leaveTangent=-4.8128, tangentMode=RCTM_User, time=1, value=0}
         ]}}}
      CurveIndex=$Particles.ParticleRelativeLifetimeSprites
      Scale Curve=&Multiply_Float(A=&RandomRangeFloat(Minimum=0.5, Maximum=1), B=1)
     )
    )
    ScaleA=true
    ScaleRGB=true
   ) {module=ScaleColor001}
   &DynamicMaterialParameters(
    Index 0 Param 1=&RandomRangeFloat(Minimum=2, Maximum=1)
    Index 0 Param 2=&RandomRangeFloat(Minimum=0, Maximum=2)
    Index 0 Param 3=8
    Index 0 Param 4=$Particles.SubImageIndex
    Index 1 Param 1=&RandomRangeFloat(Minimum=-1, Maximum=1)
    Index 1 Param 2=&RandomRangeFloat(Minimum=-1, Maximum=1)
    Index 1 Param 3=&FloatFromCurve(
     FloatCurve=DataInterface() {propertyValues {Curve {keys [
         {arriveTangent=0.8752, interpMode=RCIM_Cubic, leaveTangent=0.8752, tangentMode=RCTM_User, time=0, value=0.5}
         {arriveTangent=0.9995, interpMode=RCIM_Cubic, leaveTangent=0.9995, tangentMode=RCTM_User, time=0.2223, value=0.6825}
         {arriveTangent=-0.0405, interpMode=RCIM_Cubic, leaveTangent=-0.0405, tangentMode=RCTM_User, time=0.7, value=0.9}
         <at 1 0.7>
        ]}}}
     CurveIndex=$Particles.InterpolateCurveSprites
    )
    Index 2 Param 4=1
    Module.Param0WriteEnabled=true
    Module.Param0WriteEnabled1=true
    Module.Param0WriteEnabled2=false
    Module.Param1WriteEnabled=true
    Module.Param1WriteEnabled1=true
    Module.Param1WriteEnabled2=false
    Module.Param2WriteEnabled=true
    Module.Param2WriteEnabled1=true
    Module.Param2WriteEnabled2=false
    Module.Param3WriteEnabled=true
    Module.Param3WriteEnabled1=false
    Module.Param3WriteEnabled2=true
    Write Parameter Index 1=true
    Write Parameter Index 2=true
   )
   &SkeletalMeshLocation(Bone Sampling Mode=::NewEnumerator4, MaskSpawnGroup=false, Position Sampling=::NewEnumerator2, Skeletal Mesh=$System.Skeletal Mesh Reader)
   /SetVariables(Particles.EndPos:Vector3f=&AddVectorToPosition(
     Position=$Engine.Emitter.SimulationPosition
     Vector=&TransformVector(
      Vector=&Lerp_Vector(
       A=<-90 -100 0>
       B=<60 -100 0>
       Alpha=&FloatFromCurve(
        FloatCurve=DataInterface() {propertyValues {Curve {
           defaultValue=1
           keys [
            {arriveTangent=0.1936, interpMode=RCIM_Cubic, leaveTangent=0.1936, tangentMode=RCTM_User, time=0.4277, value=0}
            {arriveTangent=2.1083, interpMode=RCIM_Cubic, leaveTangent=2.1083, time=0.6389, value=0.3336}
            {arriveTangent=1.8454, interpMode=RCIM_Cubic, leaveTangent=1.8454, time=0.8065, value=0.7986}
            {arriveTangent=-0.2229, interpMode=RCIM_Cubic, leaveTangent=-0.2229, tangentMode=RCTM_User, time=1, value=1}
           ]
          }}}
        CurveIndex=$User.Reveal
       )
      )
      Destination Space=::World
     )
    )) {isSetParameters=true}
   /Game/Effects/Niagara/Modules/Curves/FN_HermiteSpline(
    Position Blend:NiagaraFloat=$Particles.InterpolateCurveSprites
    Position End:Vector3f=$Particles.EndPos
    Position Start:NiagaraPosition=$Particles.SkeletalMeshLocation.SampledPosition
    Write Particles Position=true
   )
   /SetVariables(Particles.Position=&MakeVector(
     X=&Multiply_Float(
      A=&VectorLength(Vector=Linked() {linkedVariable {name=StackContext.Position, type=NiagaraPosition}})
      B=&FloatFromCurve(
       FloatCurve=DataInterface() {propertyValues {Curve {
          defaultValue=1
          keys [
           {interpMode=RCIM_Cubic, leaveTangentWeight=0.1667, tangentMode=RCTM_User, time=0, value=1}
           {arriveTangentWeight=0.1667, interpMode=RCIM_Cubic, leaveTangentWeight=0.1667, tangentMode=RCTM_User, time=0.5, value=0}
           {tangentMode=RCTM_User, time=1, value=1}
          ]
         }}}
       CurveIndex=$Emitter.NormalizedLoopAge
      )
     )
     Y=&VectorLength(Vector=Linked() {linkedVariable {name=StackContext.Position, type=NiagaraPosition}})
     Z=&VectorLength(Vector=Linked() {linkedVariable {name=StackContext.Position, type=NiagaraPosition}})
    )) {enabled=false, isSetParameters=true}
  ]
  rendererClasses [&NiagaraSpriteRendererProperties]
  renderers [{rendererClass=&NiagaraSpriteRendererProperties}]
 }
 Embers {
  emitterUpdateScript [
   &EmitterState(Life Cycle Mode=::NewEnumerator1, Loop Behavior=::NewEnumerator1, UseLoopDelay=true)
   &SpawnRate(SpawnRate=50, Use Spawn Probability=false) {enabled=false, module=SpawnRate001}
   &SpawnRate(
    Spawn Probability=0.8
    SpawnRate=&Multiply_Float(
     A=100
     B=&Multiply_Float(
      A=&FloatFromCurve(
       FloatCurve=DataInterface() {propertyValues {Curve {keys [<at 0.0947 0>,<at 0.2339 0.837>,<at 0.45 1>,<at 0.8357 0.6334>,<at 1 0>]}}}
       CurveIndex=$User.Reveal
      )
      B=1
     )
    )
    Use Spawn Probability=true
   )
  ]
  particleSpawnScript [
   &InitializeParticle(
    Color=&LerpLinearColors(
     StartColor=<rgba 10 1.0866 0 1>
     EndColor=<rgba 2 0.0333 0 1>
     LerpFactor=&UniformAOrBFloat(A=0, Distribution Weight=0.2, Evaluation Type=::NewEnumerator0, Recalculate Random Each Loop=true, Override Seed=false)
    )
    Lifetime=&RandomRangeFloat(Minimum=0.33, Maximum=1)
    Sprite Size Mode=::NewEnumerator1
    Uniform Sprite Size Max=2
    Uniform Sprite Size Min=0.8
    UsePositionOffset=false
    Write Material Random=true
   )
   &SkeletalMeshLocation(Bone Sampling Mode=::NewEnumerator4, MaskSpawnGroup=false, Skeletal Mesh=$System.Skeletal Mesh Reader)
   /SetVariables(Particles.EndPos:Vector3f=&AddVectorToPosition(
     Position=$Engine.Emitter.SimulationPosition
     Vector=&TransformVector(
      Vector=&Lerp_Vector(
       A=<-90 -100 0>
       B=<60 -100 0>
       Alpha=&FloatFromCurve(
        FloatCurve=DataInterface() {propertyValues {Curve {
           defaultValue=1
           keys [
            {arriveTangent=0.1936, interpMode=RCIM_Cubic, leaveTangent=0.1936, tangentMode=RCTM_User, time=0.4277, value=0}
            {arriveTangent=2.1083, interpMode=RCIM_Cubic, leaveTangent=2.1083, time=0.6389, value=0.3336}
            {arriveTangent=1.8454, interpMode=RCIM_Cubic, leaveTangent=1.8454, time=0.8065, value=0.7986}
            {arriveTangent=-0.2229, interpMode=RCIM_Cubic, leaveTangent=-0.2229, tangentMode=RCTM_User, time=1, value=1}
           ]
          }}}
        CurveIndex=$User.Reveal
       )
      )
      Destination Space=::World
     )
    )) {enabled=false, isSetParameters=true}
   &AddVelocity(
    Cone Angle=60
    Cone Axis=&RotateVectorByQuaternion(Vector To Rotate=<1 1 0>, Quaternion=$Particles.SkeletalMeshLocation.SampledRotation)
    Module.UseConeFalloff=false
    Module.UseVelDistribution=false
    Override Seed=false
    Velocity Mode=::NewEnumerator2
    Velocity Speed=&Multiply_Float(A=&RandomRangeFloat(Minimum=333, Maximum=999), B=1)
   )
   &ShapeLocation(
    MaskSpawnGroup=false
    Offset=&TransformVector(Vector=<-3.5 11 0>, Destination Space=::World)
    Offset Mode=::NewEnumerator0
    Override Seed=false
    Sphere Radius=&Multiply_Float(A=1, B=$System.FX_Scale)
    UseOwnerScale=true
   )
  ]
  particleUpdateScript [
   &ParticleState()
   &AccelerationForce(Acceleration=&Multiply_Vector(
     A=&VectorFromFloat(Value=1)
     B=&UniformRangedVector(Minimum=<0 0 150>, Maximum=<0 0 250>)
    ))
   &Drag(
    Drag=&FloatFromCurve(
     FloatCurve=DataInterface() {propertyValues {Curve {keys [<at 0 0.5>,<at 0.5 0.5>,<at 1 1>]}}}
     Scale Curve=&RandomRangeFloat(Minimum=9, Maximum=22)
    )
    Use Linear Drag=true
    Use Rotational Drag=false
   )
   /Niagara/Modules/Update/Forces/CurlNoiseForce(
    Noise Frequency:NiagaraFloat=&RandomRangeFloat(Minimum=0.05, Maximum=0.3)
    Noise Strength=&FloatFromCurve(FloatCurve=DataInterface() {propertyValues {Curve {keys [<at 0 1>,<at 1 0>]}}}, Scale Curve=500)
    Pan Noise=true
    Pan Noise Field=<0 0 1>
   )
   &ScaleSpriteSize(Scale Factor=&Vector2DFromFloat(Value=&FloatFromCurve(FloatCurve=DataInterface() {propertyValues {Curve {keys [
          {arriveTangent=9.1531, interpMode=RCIM_Cubic, leaveTangent=9.1531, tangentMode=RCTM_User, time=0, value=0.02}
          {arriveTangent=-0.02, interpMode=RCIM_Cubic, leaveTangent=-0.02, time=0.18, value=0.99}
          <at 1 0 RCIM_Cubic>
         ]}}}))) {module=ScaleSpriteSize002}
   &GravityForce(Gravity=<0 0 -600>)
   &SolveForcesAndVelocity(Clamp Velocity=false, Limit Acceleration=false)
   &ScaleColor(
    Scale Alpha=&FloatFromCurve(
     FloatCurve=DataInterface() {propertyValues {Curve {keys [
         {arriveTangent=9.1531, interpMode=RCIM_Cubic, leaveTangent=9.1531, tangentMode=RCTM_User, time=0, value=0.02}
         {arriveTangent=0.48, interpMode=RCIM_Cubic, leaveTangent=0.48, time=0.18, value=0.99}
         <at 1 0.5 RCIM_Cubic>
        ]}}}
     Scale Curve=&MakeCustomFloatFromBool(Bool=$User.bIsFrontEndPreview, True Float=0.5, False Float=1)
    )
    Scale RGB=&MakeVectorFromLinearColorRGB(Color=&ColorFromCurve(ColorCurve=DataInterface() {propertyValues {
        AlphaCurve {keys [<at 0 1 RCIM_Cubic>]}
        BlueCurve {keys [<at 0 1.2255 RCIM_Cubic>,<at 0.5978 0 RCIM_Cubic>]}
        GreenCurve {keys [<at 0 3.7041 RCIM_Cubic>,<at 0.5978 0.0133 RCIM_Cubic>]}
        RedCurve {keys [<at 0 5 RCIM_Cubic>,<at 0.5978 0.8 RCIM_Cubic>]}
       }}))
    ScaleA=true
    ScaleRGB=true
   ) {module=ScaleColor001}
   &ScaleSpriteSizeBySpeed(
    Max Scale Factor=<1.5 5>
    Min Scale Factor=<0.4 0.8>
    Sample Scale Factor By Curve=true
    Scale Factor Curve=DataInterface() {propertyValues {Curve {keys [
        <at -0.0077 0.0765 RCIM_Cubic>
        {arriveTangent=1.2463, interpMode=RCIM_Cubic, leaveTangent=1.2463, tangentMode=RCTM_User, time=0.9985, value=0.9549}
       ]}}}
    Velocity Threshold=888
   )
   &FN_Sprite_Attributes(
    Color=&MultiplyLinearColor(Linear Color=$Particles.Color, Scale Factor=<1 1 1 0.005>)
    Create Color Attribute=true
    Create SpriteSize Attribute=true
    SpriteSize=&Multiply_Vector2D_ByFloat(Float=5, Vector2D=$Particles.SpriteSize)
    Write to Module Namespace=true
   ) {module=FN_Sprite_Attributes001}
  ]
  rendererClasses [&NiagaraSpriteRendererProperties]
  renderers [
   {rendererClass=&NiagaraSpriteRendererProperties}
   {rendererClass=&NiagaraSpriteRendererProperties, rendererIndex=1}
  ]
 }
 FireWall001 {
  emitterUpdateScript [
   &EmitterState(Life Cycle Mode=::NewEnumerator1, Loop Behavior=::NewEnumerator1, Loop Delay=0.9, Loop Duration=0.6, UseLoopDelay=true)
   &SpawnRate(
    Spawn Probability=&MakeCustomFloatFromBool(Bool=&SetBoolByFloatComparison(A=$Emitter.Age, B=0.9), True Float=0, False Float=1)
    SpawnGroup=5
    SpawnRate=30
    Use Spawn Probability=false
   ) {module=SpawnRate003}
   &SpawnRate(
    Spawn Probability=&MakeCustomFloatFromBool(Bool=&SetBoolByFloatComparison(A=$Emitter.Age, B=0.9), True Float=0, False Float=1)
    SpawnGroup=2
    SpawnRate=40
    Use Spawn Probability=false
   ) {module=SpawnRate002}
   &SpawnRate(
    Spawn Probability=&MakeCustomFloatFromBool(Bool=&SetBoolByFloatComparison(A=$Emitter.Age, B=0.9), True Float=0, False Float=1)
    SpawnGroup=1
    SpawnRate=10
    Use Spawn Probability=false
   ) {module=SpawnRate001}
   &SpawnRate(
    Spawn Probability=&MakeCustomFloatFromBool(Bool=&SetBoolByFloatComparison(A=$Emitter.Age, B=0.9), True Float=0, False Float=1)
    SpawnRate=75
    Use Spawn Probability=false
   )
   &SpawnBurst_Instantaneous(Spawn Group=4, Use Spawn Probability=false, UseLoopCountLimit=false) {module=SpawnBurst_Instantaneous003}
   &SpawnBurst_Instantaneous(Spawn Group=3, Use Spawn Probability=false, UseLoopCountLimit=false) {module=SpawnBurst_Instantaneous002}
   &SpawnBurst_Instantaneous(Spawn Count=10, Spawn Group=1, Use Spawn Probability=false, UseLoopCountLimit=false) {module=SpawnBurst_Instantaneous001}
   &SpawnBurst_Instantaneous(Spawn Count=10, Use Spawn Probability=false, UseLoopCountLimit=false)
  ]
  particleSpawnScript [
   /SetVariables(Particles.VisibilityTag=&MergeHelper_IntSwitch(
     0=&RandomRangeInt(Minimum=0, Maximum=1)
     1=2
     2=3
     3=4
     4=5
     5=6
     Index=&SpawnGroup()
     Num=::NewEnumerator4
    )) {isSetParameters=true}
   &PartitionParticles(
    Number Of Partitions=3
    Particles Per Partition:NiagaraInt32=/Niagara/DynamicInputs/Helpers/MaskInt(Mask Source:NiagaraInt32=0, Mask Parameter:NiagaraInt32=?, Mask:NiagaraInt32=4, Masked Int:NiagaraInt32=1, Passthrough Int:NiagaraInt32=5000)
    Partition Mode:ENiagara_PartitionMode_v2=::NewEnumerator1
   )
   &InitializeParticle(
    Color=&MergeHelper_ColorSwitch(
     0=<rgba 1 1 1 1>
     1=<rgba 1 1 1 1>
     2=&LerpLinearColors(
      StartColor=<rgba 10 1.0866 0 1>
      EndColor=<rgba 2 0.0333 0 1>
      LerpFactor=&UniformAOrBFloat(A=0, Distribution Weight=0.2, Evaluation Type=::NewEnumerator0, Recalculate Random Each Loop=true, Override Seed=false)
     )
     3=<rgba 1 1 1 1>
     4=<rgba 0.1 0.018 0.003 1>
     5=<rgba 1 1 1 1>
     Index=&SpawnGroup()
     Num=::NewEnumerator4
    )
    Lifetime=&MergeHelper_FloatSwitch(
     0=&RandomRangeFloat(Minimum=0.8, Maximum=1.2)
     1=&RandomRangeFloat(Minimum=1.5, Maximum=2)
     2=&RandomRangeFloat(Minimum=1.5, Maximum=2)
     3=1
     4=1.7
     5=&RandomRangeFloat(Minimum=1.5, Maximum=2)
     Index=&SpawnGroup()
     Num=::NewEnumerator4
    )
    Position Offset=<60 -100 10>
    Sprite Rotation Angle=&MergeHelper_FloatSwitch(
     0=&RandomRangeFloat(Minimum=-0.1, Maximum=0.1)
     1=&RandomRangeFloat(Minimum=0, Maximum=1)
     2=0
     3=0
     4=0
     5=&RandomRangeFloat(Minimum=0, Maximum=1)
     Index=&SpawnGroup()
     Num=::NewEnumerator4
    )
    Sprite Rotation Mode=::NewEnumerator1
    Sprite Size Mode=::NewEnumerator3
    Uniform Sprite Size=&MergeHelper_FloatSwitch(
     0=&RandomRangeFloat(Minimum=120, Maximum=150)
     1=&RandomRangeFloat(Minimum=100, Maximum=150)
     2=&RandomRangeFloat(Minimum=0.8, Maximum=2)
     3=0
     4=0
     5=&RandomRangeFloat(Minimum=3, Maximum=6)
     Index=&SpawnGroup()
     Num=::NewEnumerator4
    )
    UsePositionOffset=true
    Write Material Random=true
   )
   &ShapeLocation(
    Box Midpoint=<1 0.5 0.5>
    Box Size=&MakeVector(
     X=&FloatFromCurve(
      FloatCurve=DataInterface() {propertyValues {Curve {defaultValue=0, keys [{tangentMode=RCTM_User, time=0.828, value=1}, {tangentMode=RCTM_User, time=1, value=0}]}}}
      CurveIndex=$Emitter.NormalizedLoopAge
      Scale Curve=150
     )
     Y=10
     Z=1
    )
    MaskSpawnGroup=false
    Override Seed=false
    Shape Primitive=::NewEnumerator2
    Surface Only=false
    UseOwnerScale=true
   )
   &AddVelocity(
    Velocity=<0 0 1>
    Velocity Speed Scale=&MergeHelper_FloatSwitch(
     0=0
     1=&RandomRangeFloat(Minimum=50, Maximum=100)
     2=&RandomRangeFloat(Minimum=100, Maximum=150)
     3=0
     4=0
     5=&RandomRangeFloat(Minimum=50, Maximum=100)
     Index=&SpawnGroup()
     Num=::NewEnumerator4
    )
   )
   /SetVariables(
    Particles.DecalOrientation:Quat4f=<0 1 0 1>
    Particles.DecalPosition:NiagaraPosition=&AddVectorToPosition(Position=$Engine.Emitter.SimulationPosition, Vector=<-10 -100 0>)
   ) {isSetParameters=true}
  ]
  particleUpdateScript [
   &ParticleState()
   &AccelerationForce(Acceleration=&MergeHelper_VectorSwitch(
     0=<0 0 0>
     1=<0 0 0>
     2=&Multiply_Vector(
      A=&VectorFromFloat(Value=1)
      B=&UniformRangedVector(Minimum=<0 0 150>, Maximum=<0 0 250>)
     )
     3=?
     Index=&SpawnGroup()
     Num=::NewEnumerator1
    ))
   &ScaleSpriteSize(
    Scale Sprite Size Mode=::NewEnumerator0
    Uniform Scale Factor=&MergeHelper_FloatSwitch(
     0=&Lerp_Float(
      A=0.5
      B=0.2
      Alpha=&FloatFromCurve(
       FloatCurve=DataInterface() {propertyValues {Curve {
          defaultValue=1
          keys [
           {interpMode=RCIM_Cubic, leaveTangentWeight=0.3333, tangentMode=RCTM_User, time=0, value=0}
           {arriveTangent=1.8571, arriveTangentWeight=0.8608, interpMode=RCIM_Cubic, leaveTangent=1.8571, tangentMode=RCTM_User, time=1, value=1}
          ]
         }}}
       CurveIndex=&Multiply_Float(
        A=&VectorLength(Vector=&SubtractVectorFromPosition(
          Position=&AddVectorToPosition(Position=&SimulationPosition(), Vector=<-15 -100 10>)
          Vector=Linked() {linkedVariable {name=Particles.Position, type=NiagaraPosition}}
         ))
        B=0.015
       )
      )
     )
     1=&FloatFromCurve(FloatCurve=DataInterface() {propertyValues {Curve {keys [
          {arriveTangent=2.4634, interpMode=RCIM_Cubic, leaveTangent=2.4634, tangentMode=RCTM_User, time=0, value=0.1}
          {arriveTangent=0.2754, interpMode=RCIM_Cubic, leaveTangent=0.2754, tangentMode=RCTM_User, time=1.0463, value=1}
         ]}}})
     2=&FloatFromCurve(FloatCurve=DataInterface() {propertyValues {Curve {keys [
          {arriveTangent=9.1531, interpMode=RCIM_Cubic, leaveTangent=9.1531, tangentMode=RCTM_User, time=0, value=0.02}
          {arriveTangent=-0.02, interpMode=RCIM_Cubic, leaveTangent=-0.02, time=0.18, value=0.99}
          <at 1 0 RCIM_Cubic>
         ]}}})
     3=0
     4=0
     5=1
     Index=&SpawnGroup()
     Num=::NewEnumerator4
    )
   )
   &SubUVAnimation(
    "Sprite Renderer "=DataInterface() {propertyValues {SpriteRenderer=&self:FireWall001_0.NiagaraSpriteRendererProperties_2}}
    UseEndFrame=true
    UseStartFrame=true
   )
   &DynamicMaterialParameters(
    Index 0 Param 1=&RandomRangeFloat(Minimum=2, Maximum=1)
    Index 0 Param 2=&RandomRangeFloat(Minimum=0, Maximum=2)
    Index 0 Param 3=8
    Index 0 Param 4=$Particles.SubImageIndex
    Index 1 Param 1=&RandomRangeFloat(Minimum=-1, Maximum=1)
    Index 1 Param 2=&RandomRangeFloat(Minimum=-1, Maximum=1)
    Index 1 Param 3=&FloatFromCurve(FloatCurve=DataInterface() {propertyValues {Curve {keys [
         {arriveTangent=0.8752, interpMode=RCIM_Cubic, leaveTangent=0.8752, tangentMode=RCTM_User, time=0, value=0.5}
         {arriveTangent=0.9995, interpMode=RCIM_Cubic, leaveTangent=0.9995, tangentMode=RCTM_User, time=0.2223, value=0.6825}
         {arriveTangent=-0.0405, interpMode=RCIM_Cubic, leaveTangent=-0.0405, tangentMode=RCTM_User, time=0.7, value=0.9}
         <at 1 0.7>
        ]}}})
    Index 2 Param 4=1
    Module.Index0Float0Write=true
    Module.Index0Float1Write=true
    Module.Index0Float2Write=true
    Module.Index0Float3Write=true
    Module.Index1Float0Write=true
    Module.Index1Float1Write=true
    Module.Index1Float2Write=true
    Module.Index1Float3Write=true
    Module.Index2Float0Write=false
    Module.Index2Float1Write=false
    Module.Index2Float2Write=false
    Module.Index2Float3Write=true
    Write Parameter Index 1=::NewEnumerator1
    Write Parameter Index 2=::NewEnumerator1
   )
   /SetVariables() {isSetParameters=true}
   &Drag(
    Drag=&MergeHelper_FloatSwitch(
     0=0
     1=&RandomRangeFloat(Minimum=2, Maximum=4)
     2=&RandomRangeFloat(Minimum=1, Maximum=3)
     3=0
     4=0
     5=&RandomRangeFloat(Minimum=1, Maximum=3)
     Index=&SpawnGroup()
     Num=::NewEnumerator4
    )
    Use Linear Drag=true
    Use Rotational Drag=true
   )
   &CurlNoiseForce(
    Noise Frequency:NiagaraFloat=&MergeHelper_FloatSwitch(
     0=0
     1=50
     2=&RandomRangeFloat(Minimum=0.05, Maximum=0.3)
     3=0
     4=0
     5=&RandomRangeFloat(Minimum=0.05, Maximum=0.3)
     Index=&SpawnGroup()
     Num=::NewEnumerator4
    )
    Noise Strength=&MergeHelper_FloatSwitch(
     0=0
     1=&RandomRangeFloat(Minimum=25, Maximum=60)
     2=&FloatFromCurve(FloatCurve=DataInterface() {propertyValues {Curve {keys [<at 0 1>,<at 1 0>]}}}, Scale Curve=500)
     3=0
     4=0
     5=&FloatFromCurve(FloatCurve=DataInterface() {propertyValues {Curve {keys [<at 0 1>,<at 1 0>]}}}, Scale Curve=500)
     Index=&SpawnGroup()
     Num=::NewEnumerator4
    )
    Pan Noise=true
    Pan Noise Field=<0 0 1>
   )
   &GravityForce(Gravity=&MergeHelper_VectorSwitch(0=<0 0 0>, 1=<0 0 0>, 2=<0 0 -300>, 3=?, Index=&SpawnGroup(), Num=::NewEnumerator1))
   &SolveForcesAndVelocity(Clamp Velocity=false, Limit Acceleration=false)
   &ScaleSpriteSizeBySpeed(
    Max Scale Factor=&MergeHelper_Vector2Switch(0=<1 1>, 1=<1 1>, 2=<1.5 5>, 3=?, 4=?, Index=&SpawnGroup(), Num=::NewEnumerator1)
    Min Scale Factor=&MergeHelper_Vector2Switch(0=<1 1>, 1=<1 1>, 2=<0.4 0.8>, 3=?, 4=?, Index=&SpawnGroup(), Num=::NewEnumerator1)
    Sample Scale Factor By Curve=true
    Scale Factor Curve=DataInterface() {propertyValues {Curve {keys [
        <at -0.0077 0.0765 RCIM_Cubic>
        {arriveTangent=1.2463, interpMode=RCIM_Cubic, leaveTangent=1.2463, tangentMode=RCTM_User, time=0.9985, value=0.9549}
       ]}}}
    Velocity Threshold=888
   )
   &ScaleColor(
    Scale Alpha=&MergeHelper_FloatSwitch(
     0=1
     1=&FloatFromCurve(
      FloatCurve=DataInterface() {propertyValues {Curve {
         defaultValue=0
         keys [
          {arriveTangent=4.0743, interpMode=RCIM_Cubic, leaveTangent=4.0743, tangentMode=RCTM_User, time=0, value=0}
          {arriveTangent=3.2822, interpMode=RCIM_Cubic, leaveTangent=3.2822, time=0.0394, value=0.6082}
          {arriveTangent=0.0034, interpMode=RCIM_Cubic, leaveTangent=0.0034, tangentMode=RCTM_User, time=0.3, value=0.9847}
          {arriveTangent=-3.0662, interpMode=RCIM_Cubic, leaveTangent=-3.0662, tangentMode=RCTM_User, time=1, value=0}
         ]
        }}}
      Scale Curve=0.2
     )
     2=&FloatFromCurve(
      FloatCurve=DataInterface() {propertyValues {Curve {keys [
          {arriveTangent=9.1531, interpMode=RCIM_Cubic, leaveTangent=9.1531, tangentMode=RCTM_User, time=0, value=0.02}
          {arriveTangent=0.48, interpMode=RCIM_Cubic, leaveTangent=0.48, time=0.18, value=0.99}
          <at 1 0.5 RCIM_Cubic>
         ]}}}
      Scale Curve=&MakeCustomFloatFromBool(Bool=$User.bIsFrontEndPreview, True Float=0.5, False Float=1)
     )
     3=1
     4=&FloatFromCurve(FloatCurve=DataInterface() {propertyValues {Curve {
         defaultValue=-1
         keys [{tangentMode=RCTM_User, time=0, value=1}, <at 0.0713 -0.5924>, {tangentMode=RCTM_User, time=0.3355, value=-1}, <at 0.9939 -0.0528>]
        }}})
     5=&FloatFromCurve(FloatCurve=DataInterface() {propertyValues {Curve {
         defaultValue=1
         keys [
          {arriveTangent=9.0725, interpMode=RCIM_Cubic, leaveTangent=9.0725, tangentMode=RCTM_User, time=0, value=0}
          {arriveTangent=-0.1236, interpMode=RCIM_Cubic, leaveTangent=-0.1236, tangentMode=RCTM_User, time=0.1408, value=0.9847}
          {arriveTangent=-2.491, interpMode=RCIM_Cubic, leaveTangent=-2.491, tangentMode=RCTM_User, time=0.4109, value=0}
         ]
        }}})
     Index=&SpawnGroup()
     Num=::NewEnumerator4
    )
    Scale RGB=&MergeHelper_VectorSwitch(
     0=<1 1 1>
     1=&MakeVectorFromLinearColorRGB(Color=&ColorFromCurve(ColorCurve=DataInterface() {propertyValues {
         AlphaCurve {keys [<at 0 1 RCIM_Cubic>]}
         BlueCurve {keys [<at 0 0 RCIM_Cubic>,<at 0.1353 0.0067>,<at 0.2222 0 RCIM_Cubic>]}
         GreenCurve {keys [<at 0 0.1302 RCIM_Cubic>,<at 0.1353 0.0065>,<at 0.2222 0 RCIM_Cubic>]}
         RedCurve {keys [<at 0 0.3958 RCIM_Cubic>,<at 0.1353 0.0744>,<at 0.2222 0 RCIM_Cubic>]}
        }}))
     2=&MakeVectorFromLinearColorRGB(Color=&ColorFromCurve(ColorCurve=DataInterface() {propertyValues {
         AlphaCurve {keys [<at 0 1 RCIM_Cubic>]}
         BlueCurve {keys [<at 0 1.2255 RCIM_Cubic>,<at 0.5978 0 RCIM_Cubic>]}
         GreenCurve {keys [<at 0 3.7041 RCIM_Cubic>,<at 0.5978 0.0133 RCIM_Cubic>]}
         RedCurve {keys [<at 0 5 RCIM_Cubic>,<at 0.5978 0.8 RCIM_Cubic>]}
        }}))
     3=<1 1 1>
     4=&Multiply_Vector(
      A=&MakeVectorFromLinearColorRGB(Color=$Particles.Initial.Color)
      B=&VectorFromFloat(Value=&FloatFromCurve(FloatCurve=DataInterface() {propertyValues {Curve {
           defaultValue=1
           keys [<at 0 0>, <at 0.0709 5.75>, {tangentMode=RCTM_User, time=0.3272, value=0.9436}, {tangentMode=RCTM_User, time=0.6616, value=0.1707}, <at 1 0>]
          }}}))
     )
     5=&MakeVectorFromLinearColorRGB(Color=&ColorFromCurve(ColorCurve=DataInterface() {propertyValues {
         AlphaCurve {keys [<at 0 1 RCIM_Cubic>]}
         BlueCurve {keys [<at 0 0 RCIM_Cubic>,<at 1 0 RCIM_Cubic>]}
         GreenCurve {keys [<at 0 0.9804 RCIM_Cubic>,<at 1 0.3668 RCIM_Cubic>]}
         RedCurve {keys [<at 0 12 RCIM_Cubic>,<at 1 5 RCIM_Cubic>]}
        }}))
     Index=&SpawnGroup()
     Num=::NewEnumerator4
    )
    ScaleA=true
    ScaleRGB=true
   )
   &FN_Sprite_Attributes(
    Color=&MultiplyLinearColor(Linear Color=$Particles.Color, Scale Factor=<1 1 1 0.005>)
    Create Color Attribute=true
    Create SpriteSize Attribute=true
    SpriteSize=&Multiply_Vector2D_ByFloat(Float=5, Vector2D=$Particles.SpriteSize)
    Write to Module Namespace=true
   ) {module=FN_Sprite_Attributes001}
   &Light_Attributes(
    Color=1
    Color Value To Scale:LinearColor=&Multiply_LinearColorByFloat(Linear Color=<rgba 10 3.7559 0 1>, Float=1)
    Diffuse Scale=1
    Falloff Exponent=1
    Light Diffuse Scale=2
    Light Position Offset:Vector3f=<-10 -100 10>
    Light Radius=180.1538
    Position=1
    Position to Offset:NiagaraPosition=$Engine.Emitter.SimulationPosition
    Radius=1
    Scale Alpha:NiagaraFloat=&FloatFromCurve(FloatCurve=DataInterface() {propertyValues {Curve {
        defaultValue=1
        keys [
         {arriveTangent=9.9175, interpMode=RCIM_Cubic, leaveTangent=9.9175, tangentMode=RCTM_User, time=0, value=0}
         {arriveTangent=2.4746, interpMode=RCIM_Cubic, leaveTangent=2.4746, time=0.0549, value=0.3826}
         {arriveTangent=-0.4029, interpMode=RCIM_Cubic, leaveTangent=-0.4029, time=0.2892, value=0.7157}
         {arriveTangent=-0.0772, interpMode=RCIM_Cubic, leaveTangent=-0.0772, time=0.5311, value=0.1907}
         {arriveTangent=-0.4067, interpMode=RCIM_Cubic, leaveTangent=-0.4067, time=0.8331, value=0.6737}
         {arriveTangent=-6.7559, interpMode=RCIM_Cubic, leaveTangent=-6.7559, tangentMode=RCTM_User, time=1, value=0}
        ]
       }}})
    ScaleA=true
    ScaleRGB=true
    UsePositionOffset=true
   ) {enabled=false}
   /SetVariables(Particles.DecalFade:NiagaraFloat=&FloatFromCurve(FloatCurve=DataInterface() {propertyValues {Curve {keys [<at 0 0>,<at 0.013 0.9268>,<at 0.6624 0.9>,<at 1 0>]}}}, Scale Curve=0.5662)) {isSetParameters=true}
   /SetVariables(Particles.FlameEmbersSubImageIndex:NiagaraFloat=&FloatFromCurve(FloatCurve=DataInterface() {propertyValues {Curve {defaultValue=1, keys [{tangentMode=RCTM_User, time=0, value=0}, {tangentMode=RCTM_User, time=1, value=1}]}}})) {isSetParameters=true}
   &SpriteRotationRate(Rotation Rate=&MergeHelper_FloatSwitch(
     0=0
     1=0
     2=0
     3=0
     4=0
     5=&SelectFloatFromArray(
      Array Sampling Mode=::NewEnumerator0
      Direct Array Index=?
      Float Selection Array=DataInterface() {propertyValues {FloatData [200,-200,300,-300]}}
      Random Seed=0
     )
     Index=&SpawnGroup()
     Num=::NewEnumerator4
    )) {module=SpriteRotationRate001}
   /Niagara/Modules/Decal/Decal_Attributes(
    Color=1
    Decal Facing:Vector3f=<0 0 -1>
    Decal Facing Mode=1
    Decal Position Offset:Vector3f=&MergeHelper_VectorSwitch(0=<-40 0 0>, 1=<0 0 0>, 2=<40 0 0>, 3=?, Index=0, Num=::NewEnumerator1)
    Decal Rotation Mode=3
    Decal Size:Vector3f=<70 160 200>
    Orientation=1
    Position=1
    Position to Offset:NiagaraPosition=$Particles.DecalPosition
    Rotation Angle=0.25
    Scale Alpha:NiagaraFloat=&FloatFromCurve(FloatCurve=DataInterface() {propertyValues {Curve {defaultValue=-1, keys [{tangentMode=RCTM_User, time=0.1, value=1}, {tangentMode=RCTM_User, time=0.7, value=-1}]}}})
    Scale RGB:Vector3f=&VectorFromFloat(Value=&FloatFromCurve(FloatCurve=DataInterface() {propertyValues {Curve {
         defaultValue=1
         keys [<at 0.01 0>, <at 0.05 5.75>, {tangentMode=RCTM_User, time=0.4, value=2}, {tangentMode=RCTM_User, time=0.7, value=1}, <at 1 0>]
        }}}))
    ScaleA=true
    ScaleRGB=true
    Set Alignment=false
    Size=1
    UsePositionOffset=true
   )
  ]
  rendererClasses [
   &NiagaraSpriteRendererProperties
   &NiagaraLightRendererProperties
   &NiagaraDecalRendererProperties
  ]
  renderers [
   {rendererClass=&NiagaraSpriteRendererProperties}
   {rendererClass=&NiagaraSpriteRendererProperties, rendererIndex=1}
   {rendererClass=&NiagaraSpriteRendererProperties, rendererIndex=2}
   {rendererClass=&NiagaraSpriteRendererProperties, rendererIndex=3}
   {rendererClass=&NiagaraSpriteRendererProperties, rendererIndex=4}
   {rendererClass=&NiagaraLightRendererProperties, rendererIndex=5}
   {rendererClass=&NiagaraDecalRendererProperties, rendererIndex=6}
   {rendererClass=&NiagaraSpriteRendererProperties, rendererIndex=7}
  ]
 }
 FireWall002 {
  emitterUpdateScript [
   &EmitterState(Life Cycle Mode=::NewEnumerator1, Loop Behavior=::NewEnumerator1, Loop Delay=0.6, Loop Duration=2, UseLoopDelay=true)
   &SpawnBurst_Instantaneous(Spawn Count=25, Use Spawn Probability=false, UseLoopCountLimit=false)
  ]
  particleSpawnScript [
   &InitializeParticle(
    Color Mode=::NewEnumerator0
    Sprite Size Mode=::NewEnumerator1
    Uniform Sprite Size Max=110
    Uniform Sprite Size Min=80
    UsePositionOffset=false
    Write Material Random=true
   )
   /SetVariables(Particles.Position=&AddVectorToPosition(
     Position=$Engine.Emitter.SimulationPosition
     Vector=&TransformVector(
      Vector=&Lerp_Vector(A=<-60 -100 10>, B=<60 -100 10>, Alpha=&ReturnNormalizedExecIndex())
      Source Space=::Simulation
      Destination Space=::Simulation
     )
    )) {isSetParameters=true}
  ]
  particleUpdateScript [
   &ParticleState()
   /SetVariables(Particles.ParticleRelativeLifetimeSprites:NiagaraFloat=&Add_Float(A=&Multiply_Float(A=-1.5, B=$Emitter.Age), B=&ReturnNormalizedExecIndex())) {isSetParameters=true}
   &SubUVAnimation(
    "Sprite Renderer "=DataInterface() {propertyValues {SpriteRenderer=&self:FireWall002_0.NiagaraSpriteRendererProperties_2}}
    SubUV Lookup Index:NiagaraFloat=&ClampFloat(Float=&Subtract_Float(
      A=&Multiply_Float(A=$Particles.NormalizedAge, B=2)
      B=&FloatFromCurve(
       FloatCurve=DataInterface() {propertyValues {Curve {defaultValue=1, keys [{tangentMode=RCTM_User, time=0, value=0}, {tangentMode=RCTM_User, time=1, value=1}]}}}
       CurveIndex=&ReturnNormalizedExecIndex()
      )
     ))
    UseEndFrame=true
    UseStartFrame=true
   )
   &DynamicMaterialParameters(
    Index 0 Param 1=&RandomRangeFloat(Minimum=2, Maximum=1)
    Index 0 Param 2=&RandomRangeFloat(Minimum=0, Maximum=2)
    Index 0 Param 3=8
    Index 0 Param 4=$Particles.SubImageIndex
    Index 1 Param 1=&RandomRangeFloat(Minimum=-1, Maximum=1)
    Index 1 Param 2=&RandomRangeFloat(Minimum=-1, Maximum=1)
    Index 1 Param 3=&FloatFromCurve(FloatCurve=DataInterface() {propertyValues {Curve {keys [
         {arriveTangent=0.8752, interpMode=RCIM_Cubic, leaveTangent=0.8752, tangentMode=RCTM_User, time=0, value=0.5}
         {arriveTangent=0.9995, interpMode=RCIM_Cubic, leaveTangent=0.9995, tangentMode=RCTM_User, time=0.2223, value=0.6825}
         {arriveTangent=-0.0405, interpMode=RCIM_Cubic, leaveTangent=-0.0405, tangentMode=RCTM_User, time=0.7, value=0.9}
         <at 1 0.7>
        ]}}})
    Index 2 Param 4=1
    Module.Index0Float0Write=true
    Module.Index0Float1Write=true
    Module.Index0Float2Write=true
    Module.Index0Float3Write=true
    Module.Index1Float0Write=true
    Module.Index1Float1Write=true
    Module.Index1Float2Write=true
    Module.Index1Float3Write=true
    Module.Index2Float0Write=false
    Module.Index2Float1Write=false
    Module.Index2Float2Write=false
    Module.Index2Float3Write=true
    Write Parameter Index 1=::NewEnumerator1
    Write Parameter Index 2=::NewEnumerator1
   )
  ]
  rendererClasses [&NiagaraSpriteRendererProperties]
  renderers [{rendererClass=&NiagaraSpriteRendererProperties}]
 }
 GroundImpactFlames {
  emitterUpdateScript [
   &EmitterState(Life Cycle Mode=::NewEnumerator1, Loop Behavior=::NewEnumerator1, Loop Delay=0.27, Loop Duration=0.85, UseLoopDelay=true)
   &SpawnBurst_Instantaneous(Spawn Group=2, Use Spawn Probability=false, UseLoopCountLimit=false) {module=SpawnBurst_Instantaneous001}
   &SpawnBurst_Instantaneous(Spawn Count=3, Use Spawn Probability=false, UseLoopCountLimit=false)
   &SpawnRate(
    Spawn Probability=&MakeCustomFloatFromBool(Bool=&SetBoolByFloatComparison(A=$Emitter.Age, B=0.85), True Float=0, False Float=1)
    SpawnGroup=1
    SpawnRate=40
    Use Spawn Probability=false
   ) {module=SpawnRate001}
   &SpawnRate(
    Spawn Probability=&MakeCustomFloatFromBool(Bool=&SetBoolByFloatComparison(A=$Emitter.Age, B=0.85), True Float=0, False Float=1)
    SpawnRate=60
    Use Spawn Probability=false
   )
  ]
  particleSpawnScript [
   /SetVariables(Particles.VisibilityTag=&SpawnGroup()) {isSetParameters=true}
   /SetVariables(Particles.EndPos:Vector3f=&AddVectorToPosition(
     Position=$Engine.Emitter.SimulationPosition
     Vector=&TransformVector(
      Vector=&Lerp_Vector(
       A=<-90 -100 0>
       B=<60 -100 0>
       Alpha=&FloatFromCurve(
        FloatCurve=DataInterface() {propertyValues {Curve {
           defaultValue=1
           keys [
            {arriveTangent=0.1936, interpMode=RCIM_Cubic, leaveTangent=0.1936, tangentMode=RCTM_User, time=0.4277, value=0}
            {arriveTangent=2.1083, interpMode=RCIM_Cubic, leaveTangent=2.1083, time=0.6389, value=0.3336}
            {arriveTangent=1.8454, interpMode=RCIM_Cubic, leaveTangent=1.8454, time=0.8065, value=0.7986}
            {arriveTangent=-0.2229, interpMode=RCIM_Cubic, leaveTangent=-0.2229, tangentMode=RCTM_User, time=1, value=1}
           ]
          }}}
        CurveIndex=$User.Reveal
       )
      )
      Destination Space=::Local
     )
    )) {isSetParameters=true}
   &InitializeParticle(
    Color Mode=::NewEnumerator0
    Lifetime=&MergeHelper_FloatSwitch(
     0=&RandomRangeFloat(Minimum=0.7, Maximum=0.3)
     1=&RandomRangeFloat(Minimum=1.5, Maximum=2)
     2=2
     3=?
     Index=&SpawnGroup()
     Num=::NewEnumerator1
    )
    Position Offset=$Particles.EndPos
    Sprite Rotation Mode=::NewEnumerator2
    Sprite Size Mode=::NewEnumerator3
    Uniform Sprite Size=&MergeHelper_FloatSwitch(
     0=&RandomRangeFloat(Minimum=70, Maximum=100)
     1=&RandomRangeFloat(Minimum=80, Maximum=30)
     2=0
     3=?
     Index=&SpawnGroup()
     Num=::NewEnumerator1
    )
    UsePositionOffset=false
    Write Material Random=true
   )
   &ShapeLocation(
    Cylinder Height=1
    Cylinder Radius=10
    MaskSpawnGroup=false
    Non Uniform Scale=<1 1 0>
    Override Seed=false
    Shape Origin:Vector3f=$Particles.EndPos
    Shape Primitive=::NewEnumerator1
    Surface Only=true
    Use Endcaps In Surface Only Mode=false
    UseOwnerScale=true
   )
   &AddVelocity(
    Velocity=&MergeHelper_VectorSwitch(0=<10 -60 200>, 1=<0 -0.5 1.2>, 2=<0 0 0>, 3=?, Index=&SpawnGroup(), Num=::NewEnumerator1)
    Velocity Speed Scale=&MergeHelper_FloatSwitch(
     0=1
     1=&RandomRangeFloat(Minimum=50, Maximum=100)
     2=0
     3=?
     Index=&SpawnGroup()
     Num=::NewEnumerator1
    )
   )
   &AddVelocity(
    Velocity=$Particles.ShapeLocation.ShapeNormal
    Velocity Speed Scale=&MergeHelper_FloatSwitch(0=40, 1=0, 2=0, 3=?, Index=&SpawnGroup(), Num=::NewEnumerator1)
   ) {module=AddVelocity001}
  ]
  particleUpdateScript [
   &ParticleState()
   &ScaleSpriteSize(
    Scale Sprite Size Mode=::NewEnumerator0
    Uniform Scale Factor=&MergeHelper_FloatSwitch(
     0=&FloatFromCurve(FloatCurve=DataInterface() {propertyValues {Curve {keys [<at 0 5.7144>,<at 0.2013 3.321>,<at 0.4862 1>,<at 1 1>]}}}, CurveIndex=$Emitter.Age)
     1=&FloatFromCurve(
      FloatCurve=DataInterface() {propertyValues {Curve {keys [
          {arriveTangent=9.1531, interpMode=RCIM_Cubic, leaveTangent=9.1531, tangentMode=RCTM_User, time=0, value=0.02}
          {arriveTangent=0.5812, interpMode=RCIM_Cubic, leaveTangent=0.5812, time=0.18, value=0.99}
          <at 1 0.6012 RCIM_Cubic>
         ]}}}
      Scale Curve=1.8257
     )
     2=0
     3=?
     Index=&SpawnGroup()
     Num=::NewEnumerator1
    )
   ) {module=ScaleSpriteSize001}
   &ScaleColor(
    Scale Alpha=&MergeHelper_FloatSwitch(
     0=&FloatFromCurve(FloatCurve=DataInterface() {propertyValues {Curve {
         defaultValue=1
         keys [
          {arriveTangent=9.0725, interpMode=RCIM_Cubic, leaveTangent=9.0725, tangentMode=RCTM_User, time=0, value=0}
          {arriveTangent=-0.1236, interpMode=RCIM_Cubic, leaveTangent=-0.1236, tangentMode=RCTM_User, time=0.1408, value=0.9847}
          {arriveTangent=-2.491, interpMode=RCIM_Cubic, leaveTangent=-2.491, tangentMode=RCTM_User, time=0.4109, value=0}
         ]
        }}})
     1=&FloatFromCurve(
      FloatCurve=DataInterface() {propertyValues {Curve {
         defaultValue=0
         keys [
          {arriveTangent=4.0743, interpMode=RCIM_Cubic, leaveTangent=4.0743, tangentMode=RCTM_User, time=0, value=0}
          {arriveTangent=3.2822, interpMode=RCIM_Cubic, leaveTangent=3.2822, time=0.0394, value=0.6082}
          {arriveTangent=0.0034, interpMode=RCIM_Cubic, leaveTangent=0.0034, tangentMode=RCTM_User, time=0.3, value=0.9847}
          {arriveTangent=-1.4067, interpMode=RCIM_Cubic, leaveTangent=-1.4067, time=0.6237, value=0.3324}
          {arriveTangent=-0.6098, interpMode=RCIM_Cubic, leaveTangent=-0.6098, tangentMode=RCTM_User, time=1, value=0}
         ]
        }}}
      Scale Curve=0.3364
     )
     2=0
     3=?
     Index=&SpawnGroup()
     Num=::NewEnumerator1
    )
    Scale RGB=&MergeHelper_VectorSwitch(
     0=&MakeVectorFromLinearColorRGB(Color=&MultiplyLinearColor(
       Linear Color=&ColorFromCurve(ColorCurve=DataInterface() {propertyValues {
          AlphaCurve {keys [<at 0 1 RCIM_Cubic>]}
          BlueCurve {keys [<at 0 0.5446>, {arriveTangent=-1.2987, interpMode=RCIM_Cubic, leaveTangent=-1.2987, time=0.194, value=0}, <at 0.4193 0 RCIM_Cubic>]}
          GreenCurve {keys [<at 0 3.2766>, {arriveTangent=-5.7903, interpMode=RCIM_Cubic, leaveTangent=-5.7903, time=0.194, value=0.9804}, <at 0.4193 0.8485 RCIM_Cubic>]}
          RedCurve {keys [<at 0 12>, {arriveTangent=-16.6926, interpMode=RCIM_Cubic, leaveTangent=-16.6926, time=0.194, value=12}, <at 0.4193 5 RCIM_Cubic>]}
         }})
       Scale Factor=/Niagara/DynamicInputs/Multiply/Multiply_Vector4_ByFloat(
        Vector4:Vector4f=<1 1 1 1>
        Float:NiagaraFloat=&FloatFromCurve(FloatCurve=DataInterface() {propertyValues {Curve {keys [<at 0 5>,<at 0.2305 1>,<at 1 1>]}}}, CurveIndex=$Emitter.Age)
       )
      ))
     1=&MakeVectorFromLinearColorRGB(Color=&ColorFromCurve(ColorCurve=DataInterface() {propertyValues {
         AlphaCurve {keys [<at 0 1 RCIM_Cubic>]}
         BlueCurve {keys [<at 0 0 RCIM_Cubic>,<at 0.1353 0.0067>,<at 0.2222 0 RCIM_Cubic>]}
         GreenCurve {keys [<at 0 0.1302 RCIM_Cubic>,<at 0.1353 0.0065>,<at 0.2222 0 RCIM_Cubic>]}
         RedCurve {keys [<at 0 0.3958 RCIM_Cubic>,<at 0.1353 0.0744>,<at 0.2222 0 RCIM_Cubic>]}
        }}))
     2=<0 0 0>
     3=?
     Index=&SpawnGroup()
     Num=::NewEnumerator1
    )
    ScaleA=true
    ScaleRGB=true
   ) {module=ScaleColor002}
   /SetVariables() {isSetParameters=true}
   &ScaleVelocity(Velocity Scale=&MergeHelper_VectorSwitch(
     0=&VectorFromCurve(
      VectorCurve=DataInterface() {propertyValues {
        XCurve {keys [<at 0 1>,<at 1 1>]}
        YCurve {keys [<at 0 1>,<at 0.2665 0>,<at 0.7712 -1>]}
        ZCurve {keys [<at 0.1064 1>,<at 0.3299 1>,<at 1 1>]}
       }}
      Scale Curve=<1 10 1>
     )
     1=&VectorFromCurve(
      VectorCurve=DataInterface() {propertyValues {
        XCurve {keys [<at 0 1>,<at 1 1>]}
        YCurve {keys [<at 0 1>,<at 0.2665 0>,<at 0.7712 -1>]}
        ZCurve {keys [<at 0.1064 1>,<at 0.3299 1>,<at 1 1>]}
       }}
      Scale Curve=<1 6 1>
     )
     2=?
     3=?
     Index=&SpawnGroup()
     Num=::NewEnumerator0
    ))
   &Drag(Use Linear Drag=true, Use Rotational Drag=true)
   &ScaleVelocity(
    Coordinate Space:ENiagaraCoordinateSpace=::Simulation
    Velocity Scale=&Multiply_VectorByFloat(Vector=$Engine.Owner.SystemYAxis, Float=60)
   ) {module=ScaleVelocity001}
   /SetVariables(Particles.Velocity=&Multiply_Vector(A=$Particles.Initial.Velocity, B=$Engine.Owner.SystemYAxis)) {isSetParameters=true}
   &Light_Attributes(
    Color=1
    Color Value To Scale:LinearColor=&Multiply_LinearColorByFloat(Linear Color=<rgba 10 3.7559 0 1>, Float=1)
    Diffuse Scale=1
    Falloff Exponent=1
    Light Diffuse Scale=2
    Light Position Offset:Vector3f=<0 0 10>
    Light Radius:NiagaraFloat=&FloatFromCurve(FloatCurve=DataInterface() {propertyValues {Curve {keys [<at 0.4456 1>,<at 1 1.429>]}}}, Scale Curve=102)
    Position=1
    Position to Offset:NiagaraPosition=&AddVectorToPosition(
     Position=$Engine.Emitter.SimulationPosition
     Vector=&TransformVector(
      Vector=&Lerp_Vector(
       A=<-90 -100 0>
       B=<60 -100 0>
       Alpha=&FloatFromCurve(
        FloatCurve=DataInterface() {propertyValues {Curve {
           defaultValue=1
           keys [
            {arriveTangent=0.1936, interpMode=RCIM_Cubic, leaveTangent=0.1936, tangentMode=RCTM_User, time=0.4277, value=0}
            {arriveTangent=1.2732, interpMode=RCIM_Cubic, leaveTangent=1.2732, time=0.6389, value=0.3336}
            {arriveTangent=0.6682, interpMode=RCIM_Cubic, leaveTangent=0.6682, time=0.7974, value=0.4707}
            {arriveTangent=-0.2229, interpMode=RCIM_Cubic, leaveTangent=-0.2229, tangentMode=RCTM_User, time=1, value=0.5749}
           ]
          }}}
        CurveIndex=$User.Reveal
       )
      )
      Destination Space=::Local
     )
    )
    Radius=1
    Scale Alpha:NiagaraFloat=&FloatFromCurve(FloatCurve=DataInterface() {propertyValues {Curve {
        defaultValue=1
        keys [
         {arriveTangent=7.1586, interpMode=RCIM_Cubic, leaveTangent=7.1586, tangentMode=RCTM_User, time=0, value=0}
         {arriveTangent=-13.5964, interpMode=RCIM_Cubic, leaveTangent=-13.5964, tangentMode=RCTM_User, time=0.0899, value=2.8839}
         {arriveTangent=-0.5134, interpMode=RCIM_Cubic, leaveTangent=-0.5134, tangentMode=RCTM_User, time=0.3678, value=0.97}
         {arriveTangent=-2.4988, interpMode=RCIM_Cubic, leaveTangent=-2.4988, tangentMode=RCTM_User, time=1, value=0}
        ]
       }}})
    Scale RGB:Vector3f=&FN_Select_Vector_from_Time_Of_Day(Morning=<1 1 1>, Day=<1 1 1>, Evening=<1 1 1>, Night=<0.8 0.8 0.8>)
    ScaleA=true
    ScaleRGB=true
    UsePositionOffset=true
   )
   &SolveForcesAndVelocity(Clamp Velocity=false, Limit Acceleration=false)
  ]
  rendererClasses [&NiagaraSpriteRendererProperties, &NiagaraLightRendererProperties]
  renderers [
   {rendererClass=&NiagaraSpriteRendererProperties}
   {rendererClass=&NiagaraSpriteRendererProperties, rendererIndex=1}
   {rendererClass=&NiagaraLightRendererProperties, rendererIndex=2}
  ]
 }
 InitialGlow {
  emitterUpdateScript [
   &EmitterState(Life Cycle Mode=::NewEnumerator1, Loop Behavior=::NewEnumerator1, Loop Duration=1, UseLoopDelay=false)
   &SpawnBurst_Instantaneous(Spawn Group=2, Use Spawn Probability=false, UseLoopCountLimit=false) {module=SpawnBurst_Instantaneous002}
   &SpawnBurst_Instantaneous(Spawn Group=1, Spawn Time=0.27, Use Spawn Probability=false, UseLoopCountLimit=false) {module=SpawnBurst_Instantaneous001}
   &SpawnBurst_Instantaneous(Spawn Time=0.01, Use Spawn Probability=false, UseLoopCountLimit=false)
  ]
  particleSpawnScript [
   /SetVariables(Particles.VisibilityTag=&SpawnGroup()) {isSetParameters=true}
   &InitializeParticle(
    Color=<rgba 20 5.3 0.4 1>
    Lifetime=&MergeHelper_FloatSwitch(0=0.45, 1=0.3, 2=0.45, 3=?, Index=&SpawnGroup(), Num=::NewEnumerator1)
    Sprite Rotation Mode=::NewEnumerator1
    Sprite Size=&MergeHelper_Vector2Switch(0=<100 150>, 1=<180 180>, 2=<150 150>, 3=?, 4=?, Index=&SpawnGroup(), Num=::NewEnumerator1)
    Sprite Size Mode=::NewEnumerator4
    UsePositionOffset=false
    Write Material Random=true
   )
   &CameraOffset(Camera Offset Amount=20)
  ]
  particleUpdateScript [
   &ParticleState()
   &ScaleSpriteSize(
    Scale Factor=&MergeHelper_Vector2Switch(
     0=&Vector2DFromCurve(
      Vector2Curve=DataInterface() {propertyValues {XCurve {keys [<at 0 1>,<at 1 1>]}, YCurve {keys [<at 0 0.5789>,<at 0.2493 1>,<at 1 1>]}}}
      ScaleCurve=<1 1>
     )
     1=&Multiply_Vector2D_ByFloat(
      Float=&FloatFromCurve(FloatCurve=DataInterface() {propertyValues {Curve {
          defaultValue=1
          keys [
           {interpMode=RCIM_Cubic, tangentMode=RCTM_User, time=0, value=0.5405}
           {interpMode=RCIM_Cubic, tangentMode=RCTM_User, time=0.5, value=1}
           {tangentMode=RCTM_User, time=1, value=0}
          ]
         }}})
      Vector2D=<1 1>
     )
     2=&Multiply_Vector2D_ByFloat(
      Float=&FloatFromCurve(FloatCurve=DataInterface() {propertyValues {Curve {
          defaultValue=1
          keys [
           {interpMode=RCIM_Cubic, tangentMode=RCTM_User, time=0, value=0.5774}
           {interpMode=RCIM_Cubic, tangentMode=RCTM_User, time=0.0945, value=2.3542}
           {arriveTangent=-2.7618, interpMode=RCIM_Cubic, leaveTangent=-2.7618, time=0.2472, value=1.4545}
           {interpMode=RCIM_Cubic, leaveTangent=-1.4242, time=0.6694, value=0.7663}
          ]
         }}})
      Vector2D=<1 1>
     )
     3=?
     4=?
     Index=&SpawnGroup()
     Num=::NewEnumerator1
    )
    Scale Sprite Size Mode=::NewEnumerator2
   )
   &SkeletalMeshLocation(Bone Sampling Mode=::NewEnumerator4, MaskSpawnGroup=false, Position Sampling=::NewEnumerator2, Skeletal Mesh=$System.Skeletal Mesh Reader)
   /SetVariables(Particles.ColorGlow:LinearColor=&MultiplyLinearColor(
     Linear Color=&MergeHelper_ColorSwitch(
      0=&ColorFromCurve(ColorCurve=DataInterface() {propertyValues {
         AlphaCurve {keys [<at 0 0 RCIM_Cubic>,<at 0.1223 0.15>,<at 0.8574 0 RCIM_Cubic>]}
         BlueCurve {keys [<at 0.1587 0.5 RCIM_Cubic>,<at 0.5812 0.1018>]}
         GreenCurve {keys [<at 0.1587 2.375 RCIM_Cubic>,<at 0.5812 1.0052>]}
         RedCurve {keys [<at 0.1587 5 RCIM_Cubic>,<at 0.5812 3>]}
        }})
      1=&ColorFromCurve(ColorCurve=DataInterface() {propertyValues {
         AlphaCurve {keys [<at 0 0 RCIM_Cubic>,<at 0.1223 0.15>,<at 0.7125 0 RCIM_Cubic>]}
         BlueCurve {keys [<at 0.1587 0.5 RCIM_Cubic>,<at 0.5812 0.06>]}
         GreenCurve {keys [<at 0.1587 2.375 RCIM_Cubic>,<at 0.5812 0.501>]}
         RedCurve {keys [<at 0.1587 5 RCIM_Cubic>,<at 0.5812 3>]}
        }})
      2=&ColorFromCurve(ColorCurve=DataInterface() {propertyValues {
         AlphaCurve {keys [<at 0 0 RCIM_Cubic>,<at 0.1223 0.15>,<at 0.7125 0 RCIM_Cubic>]}
         BlueCurve {keys [<at 0.1587 1.0783 RCIM_Cubic>,<at 0.5812 0.0681>]}
         GreenCurve {keys [<at 0.1587 3.5122 RCIM_Cubic>,<at 0.5812 1.0363>]}
         RedCurve {keys [<at 0.1587 5 RCIM_Cubic>,<at 0.5812 3>]}
        }})
      3=?
      Index=&SpawnGroup()
      Num=::NewEnumerator1
     )
     Scale Factor=&MakeFloat4FromVectorAndFloat(
      "Vector (RGB)"=<1 1 1>
      "Float (Alpha)"=&MergeHelper_FloatSwitch(
       0=&Multiply_Float(A=$System.ToD, B=&FN_FrontEndFloatModifier(Gameplay=3, Front End=0.3))
       1=&Multiply_Float(A=$System.ToD, B=&FN_FrontEndFloatModifier(Gameplay=3, Front End=0.3))
       2=&Multiply_Float(A=$System.ToD, B=&FN_FrontEndFloatModifier(Gameplay=1, Front End=0.3))
       3=?
       Index=&SpawnGroup()
       Num=::NewEnumerator1
      )
     )
    )) {isSetParameters=true}
   /SetVariables(Particles.Position=&MergeHelper_VectorSwitch(
     0=&Lerp_Vector(
      A=Linked() {linkedVariable {name=Particles.SkeletalMeshLocation.SampledPosition, type=NiagaraPosition}}
      B=&AddVectorToPosition(Position=&SimulationPosition(), Vector=<-60 -100 0>)
      Alpha=&FloatFromCurve(FloatCurve=DataInterface() {propertyValues {Curve {
          defaultValue=1
          keys [{interpMode=RCIM_Cubic, tangentMode=RCTM_User, time=0, value=0}, {interpMode=RCIM_Cubic, tangentMode=RCTM_User, time=1, value=0.4678}]
         }}})
     )
     1=<-90 -100 10>
     2=&Lerp_Vector(
      A=Linked() {linkedVariable {name=Particles.SkeletalMeshLocation.SampledPosition, type=NiagaraPosition}}
      B=&AddVectorToPosition(Position=&SimulationPosition(), Vector=<-60 -100 0>)
      Alpha=0.1
     )
     3=?
     Index=&SpawnGroup()
     Num=::NewEnumerator1
    )) {isSetParameters=true}
   &SpriteFacingAndAlignment(
    Sprite Alignment=&MergeHelper_VectorSwitch(
     0=&SubtractVectorFromPosition(
      Position=$Particles.SkeletalMeshLocation.SampledPosition
      Vector=&AddVectorToPosition(Position=&SimulationPosition(), Vector=<-80 -100 0>)
     )
     1=<1 1 1>
     2=<1 0 1>
     3=?
     Index=&SpawnGroup()
     Num=::NewEnumerator1
    )
    WriteAlignment=true
    WriteFacing=false
   )
  ]
  rendererClasses [&NiagaraSpriteRendererProperties]
  renderers [
   {rendererClass=&NiagaraSpriteRendererProperties}
   {rendererClass=&NiagaraSpriteRendererProperties, rendererIndex=1}
   {rendererClass=&NiagaraSpriteRendererProperties, rendererIndex=2}
  ]
 }
 MainBurst003 {
  emitterSpawnScript [/SetVariables(Emitter.Length=166, Emitter.ParticlesToSpawn=32) {isSetParameters=true, module=SetVariables_#4}]
  emitterUpdateScript [
   &EmitterState(Inactive Response=::NewEnumerator1, Life Cycle Mode=::NewEnumerator1, Loop Behavior=::NewEnumerator1, UseLoopDelay=false)
   &SpawnBurst_Instantaneous(Spawn Count=$Emitter.ParticlesToSpawn, Use Spawn Probability=false, UseLoopCountLimit=false)
   /SetVariables(Emitter.Reveal:NiagaraFloat=$User.Reveal) {isSetParameters=true, module=SetVariables_#5}
   /SetVariables(Emitter.Length:NiagaraFloat=&Multiply_Float(
     A=$Emitter.Initial.Length
     B=&FloatFromCurve(
      FloatCurve=DataInterface() {propertyValues {Curve {keys [{interpMode=RCIM_Cubic, leaveTangent=2.4356, tangentMode=RCTM_Break, time=0, value=0}, <at 1 1>]}}}
      CurveIndex=$Emitter.Reveal
     )
    )) {isSetParameters=true, module=SetVariables_#6}
  ]
  particleSpawnScript [&InitializeParticle(Color Mode=::NewEnumerator0, Sprite Rotation Mode=::NewEnumerator2, UsePositionOffset=false, Write Material Random=true)]
  particleUpdateScript [
   &ParticleState(Kill Particles When Lifetime Has Elapsed=false)
   &SpriteRotationRate(Rotation Rate=&RandomRangeFloat(Minimum=-444, Maximum=444, Evaluation Type=::NewEnumerator1))
   /SetVariables(Particles.ParticleRelativeLifetimeRibbons:NiagaraFloat=&ReturnNormalizedExecIndex()) {isSetParameters=true, module=SetVariables_#7}
   /SetVariables(Particles.OpacityMultiplierRibbons:NiagaraFloat=&FloatFromCurve(
     FloatCurve=DataInterface() {propertyValues {Curve {
        defaultValue=0
        keys [
         {interpMode=RCIM_Cubic, tangentMode=RCTM_User, time=0, value=0}
         {interpMode=RCIM_Cubic, tangentMode=RCTM_User, time=0.5, value=1}
         {tangentMode=RCTM_User, time=1, value=0}
        ]
       }}}
     CurveIndex=&Add_Float(
      A=&Multiply_Float(A=$Particles.ParticleRelativeLifetimeRibbons, B=-1)
      B=&Multiply_Float(
       A=&FloatFromCurve(
        FloatCurve=DataInterface() {propertyValues {Curve {
           defaultValue=1
           keys [
            {arriveTangent=1.0042, interpMode=RCIM_Cubic, leaveTangent=1.0042, tangentMode=RCTM_User, time=0, value=0}
            {arriveTangent=0.5651, interpMode=RCIM_Cubic, leaveTangent=0.5651, time=0.5187, value=0.4496}
            {arriveTangent=1.1435, interpMode=RCIM_Cubic, leaveTangent=1.1435, time=0.857, value=0.4843}
            {arriveTangent=0.2076, interpMode=RCIM_Cubic, leaveTangent=0.2076, tangentMode=RCTM_User, time=1, value=1}
           ]
          }}}
        CurveIndex=$User.Reveal
       )
       B=2
      )
     )
    )) {isSetParameters=true, module=SetVariables_#8}
   /SetVariables(Particles.InterpolateCurveRibbons:NiagaraFloat=&FloatFromCurve(
     FloatCurve=DataInterface() {propertyValues {Curve {keys [
         {arriveTangent=-0.6503, interpMode=RCIM_Cubic, leaveTangent=1.1852, tangentMode=RCTM_Break, time=0, value=0}
         {arriveTangent=0.8679, interpMode=RCIM_Cubic, tangentMode=RCTM_Break, time=1, value=1}
        ]}}}
     CurveIndex=$Particles.ParticleRelativeLifetimeRibbons
    )) {isSetParameters=true, module=SetVariables_#9}
   &SkeletalMeshLocation(Bone Sampling Mode=::NewEnumerator4, MaskSpawnGroup=false, Position Sampling=::NewEnumerator2, Skeletal Mesh=$System.Skeletal Mesh Reader)
   /SetVariables(Particles.EndPos:Vector3f=&AddVectorToPosition(
     Position=$Engine.Emitter.SimulationPosition
     Vector=&TransformVector(
      Vector=&Lerp_Vector(
       A=<-90 -100 0>
       B=<60 -100 0>
       Alpha=&FloatFromCurve(
        FloatCurve=DataInterface() {propertyValues {Curve {
           defaultValue=1
           keys [
            {arriveTangent=0.1936, interpMode=RCIM_Cubic, leaveTangent=0.1936, tangentMode=RCTM_User, time=0.4277, value=0}
            {arriveTangent=2.1083, interpMode=RCIM_Cubic, leaveTangent=2.1083, time=0.6389, value=0.3336}
            {arriveTangent=1.8454, interpMode=RCIM_Cubic, leaveTangent=1.8454, time=0.8065, value=0.7986}
            {arriveTangent=-0.2229, interpMode=RCIM_Cubic, leaveTangent=-0.2229, tangentMode=RCTM_User, time=1, value=1}
           ]
          }}}
        CurveIndex=$User.Reveal
       )
      )
      Destination Space=::World
     )
    )) {isSetParameters=true}
   /SetVariables(Particles.PositionRibbons:NiagaraPosition=&Lerp_Vector(
     A=Linked() {linkedVariable {name=Particles.SkeletalMeshLocation.SampledPosition, type=NiagaraPosition}}
     B=$Particles.EndPos
     Alpha=$Particles.InterpolateCurveRibbons
    )) {isSetParameters=true, module=SetVariables_#10}
   &FN_HermiteSpline(
    Position Blend:NiagaraFloat=&Multiply_Float(A=$Emitter.Length, B=$Particles.InterpolateCurveRibbons)
    Position End:Vector3f=$Particles.EndPos
    Position Start:NiagaraPosition=$Particles.SkeletalMeshLocation.SampledPosition
    Write Particles Position=true
   ) {enabled=false}
   /SetVariables(
    Particles.BackRibbonWidth:NiagaraFloat=&FloatFromCurve(
     FloatCurve=DataInterface() {propertyValues {Curve {keys [
         {arriveTangent=1.636, interpMode=RCIM_Cubic, leaveTangent=1.636, tangentMode=RCTM_User, time=0, value=0.0777}
         {arriveTangent=1.0523, interpMode=RCIM_Cubic, leaveTangent=1.0523, tangentMode=RCTM_User, time=0.2799, value=0.5408}
         {arriveTangent=-0.0372, interpMode=RCIM_Cubic, leaveTangent=-0.0372, tangentMode=RCTM_User, time=0.7, value=0.8}
         {arriveTangent=-1.5957, interpMode=RCIM_Cubic, leaveTangent=-1.5957, tangentMode=RCTM_User, time=1, value=0.5138}
        ]}}}
     CurveIndex=$Particles.InterpolateCurveRibbons
     Scale Curve=25
    )
    Particles.ColorRibbonBack:LinearColor=&ColorFromCurve(
     ColorCurve=DataInterface() {propertyValues {
       AlphaCurve {keys [
         <at 0 0>
         <at 0.2133 1>
         {arriveTangent=-1.2718, interpMode=RCIM_Cubic, leaveTangent=-1.2718, time=0.7268, value=1}
         <at 0.9995 0 RCIM_Cubic>
        ]}
       BlueCurve {keys [<at 0.4438 0>]}
       GreenCurve {keys [<at 0.4438 0.016>]}
       RedCurve {keys [<at 0.4438 0.08>]}
      }}
     CurveIndex=$Particles.ParticleRelativeLifetimeRibbons
    )
    Particles.ColorRibbonFill:LinearColor=&MultiplyLinearColor(
     Linear Color=&ColorFromCurve(
      ColorCurve=DataInterface() {propertyValues {
        AlphaCurve {keys [
          <at 0 0>
          <at 0.2133 0.8>
          {arriveTangent=-1.0175, interpMode=RCIM_Cubic, leaveTangent=-1.0175, time=0.7268, value=0.8}
          <at 0.9995 0 RCIM_Cubic>
         ]}
        BlueCurve {keys [<at 0.4438 0>]}
        GreenCurve {keys [<at 0.4438 0.0933>]}
        RedCurve {keys [<at 0.4438 0.7>]}
       }}
      CurveIndex=$Particles.ParticleRelativeLifetimeRibbons
     )
     Scale Factor=&MakeFloat4FromVectorAndFloat(
      "Vector (RGB)"=&VectorFromFloat(Value=&FloatFromCurve(
        FloatCurve=DataInterface() {propertyValues {Curve {keys [{arriveTangent=-4.7826, interpMode=RCIM_Cubic, leaveTangent=-4.7826, tangentMode=RCTM_User, time=0.15, value=1.5}, <at 0.4 1 RCIM_Cubic>]}}}
        CurveIndex=$User.Reveal
       ))
      "Float (Alpha)"=1
     )
    )
    Particles.ColorRibbonOrange:LinearColor=&MultiplyLinearColor(
     Linear Color=&ColorFromCurve(
      ColorCurve=DataInterface() {propertyValues {
        AlphaCurve {keys [<at 0.0072 0>, <at 0.2133 0.8>, {arriveTangent=-1.032, interpMode=RCIM_Cubic, leaveTangent=-1.032, time=0.7598, value=0.8}, <at 0.9885 0>]}
        BlueCurve {keys [<at 0.4438 0.03>]}
        GreenCurve {keys [<at 0.4438 0.2505>]}
        RedCurve {keys [<at 0.4438 1.5>]}
       }}
      CurveIndex=$Particles.ParticleRelativeLifetimeRibbons
     )
     Scale Factor=&MakeFloat4FromVectorAndFloat(
      "Vector (RGB)"=&VectorFromFloat(Value=&FloatFromCurve(
        FloatCurve=DataInterface() {propertyValues {Curve {keys [
            {arriveTangent=-2379.1755, interpMode=RCIM_Cubic, leaveTangent=-2379.1755, tangentMode=RCTM_User, time=0.15, value=300}
            {arriveTangent=-161.8976, interpMode=RCIM_Cubic, leaveTangent=-161.8976, tangentMode=RCTM_User, time=0.4, value=1}
           ]}}}
        CurveIndex=$User.Reveal
       ))
      "Float (Alpha)"=1
     )
    )
    Particles.RibbonWidth=&FloatFromCurve(
     FloatCurve=DataInterface() {propertyValues {Curve {keys [
         {arriveTangent=0.5131, interpMode=RCIM_Cubic, time=0, value=0.066}
         {arriveTangent=1.2515, interpMode=RCIM_Cubic, leaveTangent=1.2515, time=0.351, value=0.3349}
         {arriveTangent=-0.003, interpMode=RCIM_Cubic, leaveTangent=-0.003, time=0.75, value=1.0046}
         {arriveTangent=-3.6992, interpMode=RCIM_Cubic, leaveTangent=-3.6992, tangentMode=RCTM_User, time=1, value=0.333}
        ]}}}
     CurveIndex=$Particles.InterpolateCurveRibbons
     Scale Curve=44
    )
   ) {isSetParameters=true, module=SetVariables_#11}
   /SetVariables(
    Particles.RibbonWidthBack:NiagaraFloat=&Multiply_Float(A=$Particles.BackRibbonWidth, B=3)
    Particles.RibbonWidthFill:NiagaraFloat=&Multiply_Float(A=$Particles.BackRibbonWidth, B=2.5)
    Particles.RibbonWidthOrange:NiagaraFloat=&Multiply_Float(A=$Particles.BackRibbonWidth, B=2.3)
   ) {isSetParameters=true}
   /SetVariables(
    Particles.ScaledColorRibbonBack:LinearColor=&MultiplyLinearColor(
     Linear Color=$Particles.ColorRibbonBack
     Scale Factor=&MakeFloat4FromVectorAndFloat("Vector (RGB)"=<1 1 1>, "Float (Alpha)"=&Multiply_Float(A=$Particles.OpacityMultiplierRibbons, B=1))
    )
    Particles.ScaledColorRibbonFill:LinearColor=&MultiplyLinearColor(
     Linear Color=$Particles.ColorRibbonFill
     Scale Factor=&MakeFloat4FromVectorAndFloat("Vector (RGB)"=<1 1 1>, "Float (Alpha)"=&Multiply_Float(A=$Particles.OpacityMultiplierRibbons, B=1))
    )
    Particles.ScaledColorRibbonOrange:LinearColor=&MultiplyLinearColor(
     Linear Color=$Particles.ColorRibbonOrange
     Scale Factor=&MakeFloat4FromVectorAndFloat("Vector (RGB)"=<1 1 1>, "Float (Alpha)"=&Multiply_Float(A=$Particles.OpacityMultiplierRibbons, B=1))
    )
   ) {isSetParameters=true}
  ]
  rendererClasses [&NiagaraRibbonRendererProperties]
  renderers [
   {rendererClass=&NiagaraRibbonRendererProperties}
   {rendererClass=&NiagaraRibbonRendererProperties, rendererIndex=1}
   {rendererClass=&NiagaraRibbonRendererProperties, rendererIndex=2}
  ]
 }
}
summary {
 systemName=NS_Emote_RareKrow_Fire
 emitters [
  {emitterName=AdditionalFlames001, rendererClasses [&NiagaraSpriteRendererProperties]}
  {emitterName=FireWall002, rendererClasses [&NiagaraSpriteRendererProperties]}
  {emitterName=MainBurst003, rendererClasses [&NiagaraRibbonRendererProperties]}
  {
   emitterName=GroundImpactFlames
   rendererClasses [&NiagaraSpriteRendererProperties, &NiagaraLightRendererProperties]
  }
  {
   emitterName=FireWall001
   rendererClasses [
    &NiagaraSpriteRendererProperties
    &NiagaraLightRendererProperties
    &NiagaraDecalRendererProperties
   ]
  }
  {emitterName=Embers, rendererClasses [&NiagaraSpriteRendererProperties]}
  {emitterName=InitialGlow, rendererClasses [&NiagaraSpriteRendererProperties]}
 ]
 userVariables [
  User.bIsFrontEnd {type=NiagaraBool}
  User.bIsFrontEndPreview {type=NiagaraBool}
  User.ExplosionDelay {type=NiagaraFloat, defaultValue=0, description=""}
  User.HitGround {type=NiagaraBool, defaultValue=true}
  User.Reveal {type=NiagaraFloat, defaultValue=0.5, description=""}
 ]
}
