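# Animation Blueprint for the RareKrow quadruped cosmetic companion, retargeting its pose from another mesh and driving its emote fire effect.
# bUseReverseRetarget selects between two RTG_RareKrow retarget branches, one adding a root yaw from the FrontEndRotation curve, which is zero outside the frontend.
# Each update, the NS_Active curve spawns or deactivates an attached Niagara fire system, while the FireReveal curve feeds its Reveal parameter.
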AnimBlueprint /CosmeticCompanions/Assets/Quadruped/RareKrow/Meshes/RareKrow_Quadruped_AnimBP
target_skeleton=&/RareKrow_Quadruped_Skeleton
graphs [
 AnimGraph {nodes [
   Copy Pose From Mesh() {bCopyCurves=true, bUseAttachedParent=true}
   n1 = "Transform (Modify) Bone - Bone: root"!(
    ComponentPose=Local To Component(LocalPose=Retarget Pose From Mesh!() -> (Pose:Pose Link Structure) {is=title AnimGraphNode_RetargetPoseFromMesh, iKRetargeterAsset=&/RTG_RareKrow})
    bAlphaBoolEnabled:bool=True
    Translation="0,0,0"
    Rotation_Roll:float=0
    Rotation_Pitch:float=0
    Rotation_Yaw:float=Get FrontEndRot() -> (FrontEndRot:double)
    Scale
   ) -> (Pose) {alphaInputType=Bool, rotationMode=BMM_Additive, rotationSpace=BCS_BoneSpace, boneToModify {boneName=root}, rotation {yaw=40}}
   Output Pose(Result=Blend Poses by bool(
     BlendPose_0=Component To Local(ComponentPose=@n1)
     BlendPose_1=Retarget Pose From Mesh() {iKRetargeterAsset=&RTG_RareKrow}
     bActiveValue=Get bUseReverseRetarget() -> (bUseReverseRetarget:bool)
    ))
  ]}
 EventGraph {nodes [
   Event BlueprintUpdateAnimation()
   > FrontEndRotation()
   > Niagara()
   TryGetPawnOwner()
  ]}
 FrontEndRotation {nodes [
   FrontEndRotation() {is=title K2Node_FunctionEntry}
   > n1 = Branch(Condition=IsInFrontend!() -> (ReturnValue:bool))
   > Set FrontEndRot(FrontEndRot:double=GetCurveValue(CurveName=FrontEndRotation)) -> (Output_Get:double)
   Set FrontEndRot(execute=@n1.else, FrontEndRot=0)
  ]}
 Niagara {nodes [
   Niagara() {is=title K2Node_FunctionEntry}
   > n1 = Branch(Condition="float > float"(A=GetCurveValue(CurveName=NS_Active)))
   > n2 = Branch(Condition=Get bNS_Spawned() -> (bNS_Spawned:bool))
   n3 = GetCurveValue(CurveName=FireReveal)
   n4 = SpawnSystemAttached!(
    execute=@n2.else
    SystemTemplate:Niagara System Object Reference=/CosmeticCompanions/FX/Builtins/RareKrow/NS_Emote_RareKrow_Fire
    AttachToComponent:Scene Component Object Reference=GetOwningComponent()
    AttachPointName:Name=root
    Location:Vector="0, 0, 0"
    Rotation:Rotator="0, 0, 0"
    LocationType:EAttachLocation Enum=SnapToTarget
    bAutoDestroy:bool=true
    bAutoActivate:bool=true
    PoolingMethod:ENCPoolMethod Enum=None
    bPreCullCheck:bool=true
   ) -> (then, ReturnValue:Niagara Particle System Component Object Reference)
   > n5 = Set SpawnedNS(SpawnedNS:Niagara Particle System Component Object Reference=@n4) -> (Output_Get:Niagara Particle System Component Object Reference)
   "Set Niagara Variable By String (Float)"!(execute=@n2.then, self:Niagara Particle System Component Object Reference=@n5, InVariableName:String=Reveal, InValue:float=@n3) -> (then) {is=title K2Node_CallFunction}
   Branch(execute=@n1.else, Condition=Get bNS_Spawned())
   > Deactivate(self:Actor Component Object Reference=Get SpawnedNS() -> (SpawnedNS:Niagara Particle System Component Object Reference))
   > Set bNS_Spawned(bNS_Spawned:bool=false) -> (Output_Get:bool)
   Set bNS_Spawned(execute=@n6.then, bNS_Spawned=true)
   n6 = "Set Niagara Variable By String (Float)"(execute=@n5.then, self=@n5, InVariableName=Reveal, InValue=@n3)
  ]}
]
variables [
 bNS_Spawned {type=boolean}
 FrontEndRot {type=number}
 SpawnedNS {type=NiagaraComponent, default=null}
]
