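# Post-process AnimBlueprint for the RareKrow quadruped cosmetic companion, adding facial posing and jaw fixes to the incoming pose.
# The input pose runs through a head facial PoseAsset blend, then component-space constraints tie jaw and reverse_jaw to head.
# Cached poses link the stages; the dynamics cache is a pass-through with no simulation nodes.
# Each update sets ArmoredVariant from the main instance's Stage2 or Stage3 outfit tags, though the AnimGraph never reads it.
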AnimBlueprint /CosmeticCompanions/Assets/Quadruped/RareKrow/Meshes/RareKrow_Quadruped_Post_AnimBP
target_skeleton=&/RareKrow_Quadruped_Skeleton
graphs [
 AnimGraph {nodes [
   Output Pose(Result=Use cached pose 'DynamicsCache'())
   Save cached pose 'CRCache_Input'(Pose=Pose!(
     PoseAsset:Pose Asset Object Reference=Get PoseAsset() -> (PoseAsset:Pose Asset Object Reference)
     SourcePose:Pose Link Structure=Input Pose()
    ) -> (Pose:Pose Link Structure) {is=title AnimGraphNode_PoseBlendNode, bIgnorePoseAssetSourceAnimationCurves=true, blendOption=Linear}) {cacheName=CRCache_Input}
   Save cached pose 'DynamicsCache'(Pose=Use cached pose 'PreDynamicsCache'()) {cacheName=DynamicsCache}
   n1 = "Constraint - jaw (P)"!(ComponentPose=Local To Component(LocalPose=Use cached pose 'CRCache_Input'()), bAlphaBoolEnabled:bool=True, ConstraintWeights_0=1) -> (Pose) {alphaInputType=Bool, boneToModify {boneName=jaw}, constraintSetup [{targetBone {boneName=head}}], constraintWeights [1]}
   n2 = "Constraint - reverse_jaw (P)"!(ComponentPose=@n1, bAlphaBoolEnabled:bool=True, ConstraintWeights_0=1) -> (Pose) {alphaInputType=Bool, boneToModify {boneName=reverse_jaw}, constraintSetup [{targetBone {boneName=head}}], constraintWeights [1]}
   Save cached pose 'PreDynamicsCache'(Pose=Component To Local(ComponentPose=@n2)) {cacheName=PreDynamicsCache}
  ]}
 EventGraph {nodes [
   Event BlueprintUpdateAnimation()
   n1 = GetMainAnimInstance!(self:Companion Post Process Anim Instance Object Reference) -> (ReturnValue:Fort Player Companion Anim Instance Object Reference)
   > Set ArmoredVariant(ArmoredVariant:bool=OR Boolean(
     A=HasVariantTagForChannel!(
      self:Fort Player Companion Anim Instance Object Reference=@n1
      ChannelTag:Gameplay Tag Structure=(TagName="Cosmetics.Variant.Channel.Outfit")
      VariantTag:Gameplay Tag Structure=(TagName="Cosmetics.Variant.Property.Stage2")
     ) -> (ReturnValue:bool)
     B=HasVariantTagForChannel(self=@n1, ChannelTag=(TagName="Cosmetics.Variant.Channel.Outfit"), VariantTag=(TagName="Cosmetics.Variant.Property.Stage3"))
    )) -> (Output_Get:bool)
  ]}
]
variables [
 ArmoredVariant {type=boolean}
 PoseAsset {type=PoseAsset, default=&/RareKrow_Quadruped_Head_Facial_Poses_PoseAsset}
]
